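# Translucent unlit mesh-particle material that, judging by its name, draws a motion-blur streak effect for the Cube.
# A parabolic mask along U, peaking at the center, multiplies stretched noise whose UVs are shifted by a dynamic UVOffset parameter, with both terms shaped by Power nodes.
# The result is scaled by particle alpha, softened by a parameterized DepthFade and multiplied by inverted Fresnel to fade grazing-angle edges.
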Material /Figment_S06/Athena/Prototype/Blueprints/Cube/FX/M_Cube_MotionBlur
graph {
 properties {bUsedWithMeshParticles=true, blendMode=BLEND_Translucent, pixelDepthOffsetMode=PDOM_Legacy, shadingModel=MSM_Unlit, usedShadingModels=Unlit}
 nodes [
  n1 = ParticleColor()
  n2 = MaterialFunctionCall(Float2=TextureCoordinate()) {materialFunction=&BreakOutFloat2Components}
  n3 = Multiply(A=@n2.R, B=OneMinus(None=@n2.R))
  n4 = Add(A=TextureCoordinate() {uTiling=0.05, vTiling=1.5}, B=DynamicParameter() {paramNames [UVOffset,Param2,Param3,Param4]}.UVOffset)
  n5 = Multiply(
   A=Power(Base=Multiply(A=@n3) {constB=4})
   B=Power(Base=TextureSample(UVs=@n4) {texture=/Figment_S06/Effects/Fort_Effects/Gameplay/SmokeDustExplosions/Textures/T_NoisePack_01}.G)
  )
  n6 = Fresnel() {exponent=1}
  Multiply(
   A=DepthFade(Opacity=Multiply(A=@n1.A, B=@n5), FadeDistance=ScalarParameter() {defaultValue=200, parameterName=DepthFade})
   B=Power(Base=OneMinus(None=@n6)) {constExponent=1.5}
  )
 ]
}
