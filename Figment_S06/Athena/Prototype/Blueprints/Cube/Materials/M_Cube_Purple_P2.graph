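# Two-sided lit material for the purple Cube prototype, with a pulsing inner glow, parallax runes and glowing corners.
# An analytic camera-ray-versus-sphere test fakes an inner orb, height-blended with cube noise.
# Bump-offset DFHeightMask layers and panning noise form runes, added to corner and world-height gradient glows.
# RuneActive switches glow states, Zap parameters boost inner glow, and quality switches supply cheaper Low variants.
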Material /Figment_S06/Athena/Prototype/Blueprints/Cube/Materials/M_Cube_Purple_P2
graph {
 properties {pixelDepthOffsetMode=PDOM_Legacy, twoSided=true, usedShadingModels=Default Lit}
 nodes [
  n1 = MaterialFunctionCall(Value=Multiply(A=Time(), B=ScalarParameter() {defaultValue=1, parameterName=PulseSpeed_3})) {materialFunction=&LinearSine}
  n2 = Multiply(
   A=LinearInterpolate(A=Constant() {r=1200}, B=Constant() {r=1400}, Alpha=@n1."Rounded Linear Sine ")
   B=ScalarParameter() {defaultValue=1, parameterName=RadiusMult}
  )
  n3 = ObjectPositionWS()
  n4 = VectorParameter() {parameterName=offz, defaultValue=<rgba 0 0 200 1>}
  n5 = Subtract(A=CameraPositionWS(), B=Add(A=Add(A=@n3, B=@n4.RGB), B=VectorParameter() {desc=For zapping players, parameterName=SphereOffset}.RGB))
  n6 = Distance(A=@n5, B=Constant())
  n7 = Multiply(A=@n6, B=@n6)
  n8 = DotProduct(A=@n5, B=CameraVectorWS())
  n9 = Multiply(A=@n8, B=@n8)
  n10 = Max(A=Subtract(A=Multiply(A=@n2, B=@n2) {desc="r^2"}, B=Subtract(A=@n7, B=@n9))) {constB=0}
  n11 = Divide(A=Multiply(A=SquareRoot(None=@n10)) {constB=2}, B=Multiply(A=@n2) {constB=2})
  n12 = Power(Base=@n11, Exp=ScalarParameter() {defaultValue=4, parameterName=4444235_2})
  n13 = Multiply(A=Constant() {r=3.5}, B=TextureCoordinate())
  n14 = TextureSample(UVs=@n13) {samplerType=SAMPLERTYPE_Masks, texture=/Game/Global/Textures/Default/Noise/T_CubeNoise}
  n15 = Saturate(None=Add(A=@n14.R, B=ScalarParameter() {defaultValue=0.25, parameterName=4444235}))
  n16 = LinearInterpolate(
   A=VectorParameter() {parameterName=GlowColor_One_LQ_1, defaultValue=<rgba 0.0178 0.0005 0.1 1>}.RGB
   B=VectorParameter() {parameterName=GlowColor_One_LQ_2, defaultValue=<rgba 0.2218 0.024 0.3 1>}.RGB
   Alpha=MaterialFunctionCall(Transition Phase=@n12, Height Texture=@n15) {materialFunction=&HeightLerp}.Lerp Alpha No Contrast
  )
  n17 = ScalarParameter() {parameterName=G_14}
  n18 = Saturate(None=Add(A=ScalarParameter() {defaultValue=-0.5, parameterName=G_13}, B=@n12))
  Add(A=@n16, B=Multiply(A=@n16, B=Multiply(A=@n17, B=@n18)))
  Constant3Vector() {constant=<rgba 0 0 -400 1>}
  Constant() {r=16}
  Constant() {r=0.0002}
  Constant() {r=2.5}
  Constant() {r=-1000}
  Constant() {r=1}
  Constant() {r=0.0135}
  Constant() {r=-0.2}
  Constant() {r=5}
  Constant() {r=1}
  Constant() {r=4}
  n19 = TextureCoordinate() {coordinateIndex=2}
  n20 = Constant2Vector() {g=0.5, r=0.5}
  n21 = MaterialFunctionCall(Float2=Abs(None=Subtract(A=@n19, B=@n20))) {materialFunction=&BreakOutFloat2Components}
  n22 = Max(A=@n21.R, B=@n21.R)
  n23 = Multiply(A=@n22, B=Constant() {r=2})
  MaterialFunctionCall(Transition Phase=Power(Base=Saturate(None=@n23), Exp=Constant() {r=1.8}), Height Texture=@n14.R) {materialFunction=&HeightLerp}
  n24 = Multiply(A=@n22, B=Constant() {r=1.4})
  MaterialFunctionCall(Transition Phase=Power(Base=Saturate(None=@n24), Exp=Constant() {r=1.5}), Height Texture=@n14.R) {materialFunction=&HeightLerp}
  n25 = VectorParameter() {parameterName=GlowColor_One, defaultValue=<rgba 0.5374 0 0.75 1>}
  QualitySwitch(Default=@n25.RGB, Low=VectorParameter() {parameterName=GlowColor_One_LQ, defaultValue=<rgba 0.5374 0 0.75 1>}.RGB, High=@n25.RGB)
  n26 = Multiply(A=@n22, B=Constant() {r=1.95})
  n27 = MaterialFunctionCall(Transition Phase=Power(Base=Saturate(None=@n26), Exp=Constant() {r=5}), Height Texture=@n14.R) {materialFunction=&HeightLerp}
  QualitySwitch(Default=@n27.Lerp Alpha No Contrast, Low=@n14.R, High=@n27.Lerp Alpha No Contrast)
  n28 = Saturate(None=Add(A=@n27.Lerp Alpha No Contrast, B=@n14.R))
  QualitySwitch(Default=@n28, Low=@n14.R, High=@n28)
  n29 = VectorParameter() {parameterName=BaseDiffuse, defaultValue=<rgba 0.015 0.0013 0.1 1>}
  n30 = LinearInterpolate(A=@n29.RGB, B=VectorParameter() {parameterName=EdgeDiffuse, defaultValue=<rgba 0.006 0.0005 0.04 1>}.RGB, Alpha=@n27.Lerp Alpha No Contrast)
  QualitySwitch(Default=@n30, Low=@n29.RGB, High=@n30)
  Reroute(None=LinearInterpolate(A=Constant() {r=0.29}, B=Constant() {r=0.8}, Alpha=@n14.R))
  n31 = VertexColor()
  n32 = TwoSidedSign()
  n33 = Subtract(A=WorldPosition().XYZ, B=Add(A=ObjectPositionWS(), B=ScalarParameter() {defaultValue=-1200, parameterName=ZGrad-Offset}))
  n34 = ScalarParameter() {defaultValue=0.0002, parameterName=ZGrad-Size}
  n35 = Power(Base=ComponentMask(None=Multiply(A=@n33, B=@n34)) {b=true}, Exp=ScalarParameter() {defaultValue=2.5, parameterName=ZGrad-Power})
  n36 = OneMinus(None=@n22)
  n37 = ScalarParameter() {defaultValue=0.25, parameterName=ZGrad-Mod_1}
  n38 = TextureSample(UVs=BumpOffset(Coordinate=@n13, Height=Constant() {r=0.58}, HeightRatioInput=Constant() {r=-0.5})) {samplerType=SAMPLERTYPE_Masks, texture=&T_CubeNoise}
  n39 = QualitySwitch(Default=@n38.G, Low=@n14.G, High=@n38.G)
  n40 = Multiply(A=Multiply(A=Multiply(A=@n36, B=@n35), B=ScalarParameter() {defaultValue=50, parameterName=ZGrad-Mod}), B=Saturate(None=Add(A=@n37, B=@n39)))
  n41 = SphereMask(A=TextureCoordinate() {coordinateIndex=2}, B=Constant() {r=0.5}, Radius=Constant() {r=1}, Hardness=Constant()) {attenuationRadius=0.7, hardnessPercent=30}
  n42 = Multiply(A=@n36, B=@n41)
  n43 = VectorParameter() {parameterName=CornerGlowColor, defaultValue=<rgba 20.105 0 25 1>}
  n44 = Multiply(
   A=Power(Base=OneMinus(None=@n42), Exp=ScalarParameter() {defaultValue=16, parameterName=CornerGlow-Power_2})
   B=QualitySwitch(Default=@n43.RGB, Low=VectorParameter() {parameterName=CornerGlowColor_LQ, defaultValue=<rgba 17.2588 6 20 1>}.RGB, High=@n43.RGB)
  )
  n45 = ScalarParameter() {parameterName=RuneActive}
  n46 = LinearInterpolate(
   A=ScalarParameter() {defaultValue=0.3, parameterName=EdgeGlowWhileOff}
   B=ScalarParameter() {defaultValue=1, parameterName=EdgeGlowWhileOn}
   Alpha=@n45
  )
  n47 = ScalarParameter() {parameterName=Zap}
  n48 = ScalarParameter() {parameterName=Zap2}
  n49 = Add(A=Add(A=Multiply(A=@n47) {constB=2}, B=Multiply(A=@n48) {constB=1.5}), B=@n45)
  n50 = LinearInterpolate(
   A=ScalarParameter() {defaultValue=1, parameterName=InnerGlowWhileOff_1}
   B=ScalarParameter() {defaultValue=3, parameterName=InnerGlowWhileOn_1}
   Alpha=@n49
  )
  n51 = Power(
   Base=@n39
   Exp=LinearInterpolate(
    A=ScalarParameter() {defaultValue=2.75, parameterName=4444235_3}
    B=ScalarParameter() {defaultValue=1.5, parameterName=4444235_4}
    Alpha=@n1."Rounded Linear Sine "
   )
  )
  n52 = ScalarParameter() {defaultValue=0.4, parameterName=SMaskRad1_1}
  n53 = Multiply(A=@n36, B=@n41)
  n54 = Multiply(
   A=Saturate(None=Add(A=@n52, B=@n53))
   B=MaterialFunctionCall(Transition Phase=@n11, Height Texture=Multiply(A=@n15, B=@n51)) {materialFunction=&HeightLerp}.Lerp Alpha No Contrast
  )
  n55 = ScalarParameter() {defaultValue=2, parameterName=gpow3_2}
  n56 = Saturate(None=Add(A=ScalarParameter() {defaultValue=-0.5, parameterName=gpow3_1}, B=Multiply(A=@n55, B=@n53)))
  n57 = LinearInterpolate(
   A=ScalarParameter() {defaultValue=-0.5, parameterName=deactivated}
   B=ScalarParameter() {defaultValue=1, parameterName=activated}
   Alpha=@n45
  )
  n58 = OneMinus(None=@n39)
  n59 = ScalarParameter() {defaultValue=3, parameterName=G_10}
  n60 = MaterialFunctionCall(Float2=Multiply(A=ScalarParameter() {defaultValue=19, parameterName=BB_3}, B=TextureCoordinate() {vTiling=0.75})) {materialFunction=&BreakOutFloat2Components}
  n61 = MaterialFunctionCall(
   X=Add(A=ScalarParameter() {defaultValue=0.1, parameterName=BB_13}, B=@n60.R)
   Y=Add(A=@n60.G, B=ScalarParameter() {defaultValue=0.05, parameterName=BB_14})
  ) {materialFunction=&MakeFloat2}
  n62 = LinearInterpolate(
   A=Constant() {r=-1}
   B=Constant() {r=1}
   Alpha=TextureSample(UVs=@n61.Result) {samplerType=SAMPLERTYPE_LinearColor, texture=/Game/Athena/Prototype/Blueprints/Cube/Materials/DFHeightMask}.G
  ) {constA=-1}
  n63 = ScalarParameter() {defaultValue=0.5, parameterName=qw1_1}
  n64 = TextureSample(UVs=BumpOffset(Coordinate=@n61.Result, Height=Multiply(A=@n62, B=@n63), HeightRatioInput=ScalarParameter() {defaultValue=0.15, parameterName=BB_8})) {samplerType=SAMPLERTYPE_LinearColor, texture=&DFHeightMask}
  n65 = ScalarParameter() {defaultValue=5, parameterName=G_2}
  n66 = Multiply(A=ScalarParameter() {defaultValue=32, parameterName=BB_2}, B=TextureCoordinate())
  n67 = BumpOffset(
   Coordinate=@n66
   Height=LinearInterpolate(
    A=ScalarParameter() {defaultValue=-1, parameterName=qw1_8}
    B=ScalarParameter() {defaultValue=1, parameterName=qw1_9}
    Alpha=TextureSample(UVs=@n66) {samplerType=SAMPLERTYPE_LinearColor, texture=&DFHeightMask}.R
   )
   HeightRatioInput=ScalarParameter() {defaultValue=0.06, parameterName=BB_4}
  )
  n68 = TextureSample(UVs=QualitySwitch(Default=@n67, Low=@n66, High=@n67)) {samplerType=SAMPLERTYPE_LinearColor, texture=&DFHeightMask}
  n69 = Power(
   Base=@n68.R
   Exp=LinearInterpolate(
    A=ScalarParameter() {defaultValue=1.5, parameterName=G_6}
    B=ScalarParameter() {defaultValue=5, parameterName=G_1}
    Alpha=@n1."Rounded Linear Sine "
   )
  )
  n70 = ScalarParameter() {defaultValue=5, parameterName=G_9}
  n71 = ScalarParameter() {defaultValue=1, parameterName=qw1_3}
  n72 = ScalarParameter() {defaultValue=-2, parameterName=PanningSquareNoise-Contrast}
  n73 = QualitySwitch(Default=@n72, Low=ScalarParameter() {defaultValue=-0.75, parameterName=LQ-PanningSquareNoise-Contrast}, High=@n72)
  n74 = Constant2Vector() {r=0.011}
  n75 = Multiply(A=Constant() {r=2}, B=TextureCoordinate())
  n76 = ScalarParameter() {defaultValue=0.0312, parameterName=BB_7}
  n77 = Time()
  n78 = TextureSample(UVs=Panner(Coordinate=Add(A=@n74, B=@n75), Time=Multiply(A=@n76, B=@n77)) {speedX=1}) {samplerType=SAMPLERTYPE_LinearColor, texture=&DFHeightMask}
  n79 = Multiply(A=@n68.R, B=Saturate(None=LinearInterpolate(A=@n71, B=@n73, Alpha=@n78.B)))
  n80 = Multiply(A=@n79, B=ScalarParameter() {defaultValue=1.25, parameterName=G_3})
  n81 = Multiply(
   A=Multiply(A=Power(Base=@n58, Exp=@n59), B=ScalarParameter() {defaultValue=10, parameterName=4444235_5})
   B=LinearInterpolate(A=Power(Base=@n64.G, Exp=@n65), B=Multiply(A=@n69, B=@n70), Alpha=Saturate(None=@n80))
  )
  n82 = Multiply(A=@n56, B=Multiply(A=@n57, B=@n81))
  n83 = LinearInterpolate(A=@n82, B=Constant(), Alpha=@n45)
  n84 = MaterialFunctionCall(
   1=VectorParameter() {parameterName=GlowColor_4, defaultValue=<rgba 0.045 0.004 0.3 1>}.RGB
   2=VectorParameter() {parameterName=GlowColor_3, defaultValue=<rgba 0.0712 0.002 0.4 1>}.RGB
   3=VectorParameter() {parameterName=GlowColor_2, defaultValue=<rgba 0.2547 0 0.5 1>}.RGB
   4=@n25.RGB
   A=Saturate(None=Add(A=@n54, B=@n83))
  ) {materialFunction=&Lerp_Multiple_Float3}
  n85 = Add(
   A=Multiply(A=@n51, B=ScalarParameter() {defaultValue=0.5, parameterName=4444235_1})
   B=Power(Base=@n11, Exp=ScalarParameter() {defaultValue=15, parameterName=4444235_6})
  )
  n86 = Multiply(A=Multiply(A=@n56, B=@n45), B=@n81)
  n87 = Add(
   A=LinearInterpolate(A=@n84.Lerp 4 Inputs, B=Multiply(A=@n50, B=@n84.Lerp 4 Inputs), Alpha=Saturate(None=@n85))
   B=Multiply(A=VectorParameter() {parameterName=GlowColor_Rune, defaultValue=<rgba 1 0 0.8943 1>}.RGB, B=Multiply(A=@n82, B=@n86))
  )
  n88 = LinearInterpolate(
   A=Multiply(A=ScalarParameter() {defaultValue=0.25, parameterName=InnerGlowWhileOff}, B=@n87)
   B=Multiply(A=@n87, B=ScalarParameter() {defaultValue=1, parameterName=InnerGlowWhileOn})
   Alpha=@n49
  )
  n89 = VectorParameter() {parameterName=GlowColor_Rune_LQ, defaultValue=<rgba 0.35 0 0.3438 1>}
  n90 = ScalarParameter() {defaultValue=10, parameterName=G_15}
  n91 = ScalarParameter() {defaultValue=2, parameterName=G_16}
  n92 = ScalarParameter() {defaultValue=4, parameterName=G_4}
  n93 = Add(A=ScalarParameter() {defaultValue=-0.3, parameterName=G_11}, B=@n58)
  n94 = Multiply(
   A=Multiply(A=LinearInterpolate(A=@n90, B=@n91, Alpha=@n1."Rounded Linear Sine "), B=Power(Base=@n79, Exp=@n92))
   B=Multiply(A=ScalarParameter() {defaultValue=8, parameterName=G_12}, B=Saturate(None=@n93))
  )
  n95 = Add(
   A=Multiply(A=ScalarParameter() {defaultValue=2, parameterName=InnerGlowWhileOn_2}, B=@n16)
   B=Multiply(A=Multiply(A=@n89.RGB, B=@n94), B=Multiply(A=@n56, B=@n57))
  )
  n96 = Multiply(A=ScalarParameter() {defaultValue=0.1, parameterName=InnerGlowWhileOff-LQ}, B=@n95)
  n97 = Multiply(A=@n95, B=Constant() {r=1})
  n98 = Add(
   A=Add(A=Multiply(A=@n40, B=@n25.RGB), B=Multiply(A=@n44, B=@n46))
   B=QualitySwitch(Default=@n88, Low=LinearInterpolate(A=@n96, B=@n97, Alpha=@n49), High=@n88)
  )
  Reroute(None=Multiply(A=@n31.R, B=Multiply(A=@n32, B=@n98)))
  n99 = LinearInterpolate(A=Constant() {r=0.2}, B=Constant() {r=0.1}, Alpha=@n14.R)
  Reroute(None=Multiply(A=@n31.R, B=Multiply(A=@n32, B=@n99)))
  Reroute(None=Multiply(A=@n31.R, B=Multiply(A=@n29.RGB, B=@n32)))
  Reroute(None=Constant() {r=0.2})
 ]
}
