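# Two-sided lit material for the prototype purple Cube, with glowing corners, a noisy height gradient and an inner glow.
# UV2 edge masks and bump-offset noise shape the glows, while an analytic camera-ray sphere intersection fakes an orb inside.
# Time sines drive pulsing, MPC_Cube's CubeGrowth_Unstable shifts colors and pulse, and Zap parameters boost the inner glow.
# Quality switches simplify low-end shading; vertex red and two-sided sign mask the results.
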Material /Figment_S06/Athena/Prototype/Blueprints/Cube/Materials/M_Cube_Purple_P2_2
graph {
 properties {pixelDepthOffsetMode=PDOM_Legacy, twoSided=true, usedShadingModels=Default Lit}
 nodes [
  n1 = Add(A=ObjectPositionWS(), B=VectorParameter() {parameterName=offz, defaultValue=<rgba 0 0 200 1>}.RGB)
  Add(A=@n1, B=VectorParameter() {desc=For zapping players, parameterName=SphereOffset}.RGB)
  n2 = ScalarParameter() {defaultValue=0.77, parameterName="Speed (S)_5"}
  n3 = ScalarParameter() {defaultValue=2, parameterName="Speed (S)_3"}
  n4 = Subtract(A=WorldPosition().XYZ, B=Add(A=ObjectPositionWS(), B=ScalarParameter() {defaultValue=-1200, parameterName=ZGrad-Offset}))
  n5 = ScalarParameter() {defaultValue=0.0002, parameterName=ZGrad-Size}
  n6 = Power(Base=ComponentMask(None=Multiply(A=@n4, B=@n5)) {b=true}, Exp=ScalarParameter() {defaultValue=2.5, parameterName=ZGrad-Power})
  n7 = TextureCoordinate() {coordinateIndex=2}
  n8 = Constant2Vector() {g=0.5, r=0.5}
  n9 = MaterialFunctionCall(Float2=Abs(None=Subtract(A=@n7, B=@n8))) {materialFunction=&BreakOutFloat2Components}
  n10 = OneMinus(None=Max(A=@n9.R, B=@n9.R))
  n11 = ScalarParameter() {defaultValue=0.25, parameterName=ZGrad-Mod_1}
  n12 = Multiply(A=Constant() {r=3.5}, B=TextureCoordinate())
  n13 = TextureSample(UVs=BumpOffset(Coordinate=@n12, Height=Constant() {r=0.58}, HeightRatioInput=Constant() {r=-0.5})) {samplerType=SAMPLERTYPE_Masks, texture=&/T_CubeNoise}
  n14 = TextureSample(UVs=@n12) {samplerType=SAMPLERTYPE_Masks, texture=&T_CubeNoise}
  n15 = QualitySwitch(Default=@n13.G, Low=@n14.G, High=@n13.G)
  n16 = Multiply(A=Multiply(A=Multiply(A=@n10, B=@n6), B=ScalarParameter() {defaultValue=50, parameterName=ZGrad-Mod}), B=Saturate(None=Add(A=@n11, B=@n15)))
  n17 = MaterialFunctionCall(
   A=MaterialFunctionCall(Speed=ScalarParameter() {defaultValue=5, parameterName="Speed (S)"}) {materialFunction=&DebugTimeSine}.Result
   B=MaterialFunctionCall(Speed=ScalarParameter() {defaultValue=2, parameterName="Speed (S)_1"}) {materialFunction=&DebugTimeSine}.Result
   C=MaterialFunctionCall(Speed=ScalarParameter() {defaultValue=2, parameterName="Speed (S)_2"}) {materialFunction=&DebugTimeSine}.Result
   Alpha=LinearInterpolate(
    A=MaterialFunctionCall(Speed=@n2) {materialFunction=&DebugTimeSine}.Result
    B=MaterialFunctionCall(Speed=@n3) {materialFunction=&DebugTimeSine}.Result
    Alpha=@n16
   )
  ) {materialFunction=&3ColorBlend}
  Clamp(None=@n17.Result)
  Constant3Vector() {constant=<rgba 0 0 -400 1>}
  Constant() {r=16}
  Constant() {r=0.0002}
  Constant() {r=2.5}
  Constant() {r=-1000}
  Constant() {r=1200}
  Constant() {r=1400}
  Reroute(None=LinearInterpolate(A=Constant() {r=0.29}, B=Constant() {r=0.8}, Alpha=@n14.R))
  n18 = VertexColor()
  n19 = TwoSidedSign()
  n20 = CollectionParameter() {collection=/Game/Athena/Prototype/Blueprints/Cube/Materials/MPC_Cube, parameterName=CubeGrowth_Unstable}
  n21 = LinearInterpolate(
   A=VectorParameter() {parameterName=GlowColor_One, defaultValue=<rgba 0.4299 0 0.6 1>}.RGB
   B=VectorParameter() {parameterName=GlowColor_Unstable_One, defaultValue=<rgba 0.7166 0 1 1>}.RGB
   Alpha=@n20
  )
  n22 = SphereMask(A=TextureCoordinate() {coordinateIndex=2}, B=Constant() {r=0.5}, Radius=Constant() {r=1}, Hardness=Constant()) {attenuationRadius=0.7, hardnessPercent=30}
  n23 = Power(Base=OneMinus(None=Multiply(A=@n10, B=@n22)), Exp=ScalarParameter() {defaultValue=16, parameterName=CornerGlow-Power_2})
  n24 = Time()
  n25 = ScalarParameter() {defaultValue=1.4, parameterName=PulseSpeed_3}
  n26 = LinearInterpolate(
   A=MaterialFunctionCall(Value=Multiply(A=@n24, B=@n25)) {materialFunction=&LinearSine}."Rounded Linear Sine "
   B=Multiply(A=Saturate(None=@n17.Result), B=ScalarParameter() {defaultValue=1, parameterName=CubeGrowth_Unstable_Mod})
   Alpha=@n20
  )
  n27 = VectorParameter() {parameterName=CornerGlowColor, defaultValue=<rgba 20.105 0 25 1>}
  n28 = Multiply(
   A=Add(A=@n23, B=Multiply(A=@n23, B=@n26))
   B=QualitySwitch(Default=@n27.RGB, Low=VectorParameter() {parameterName=CornerGlowColor_LQ, defaultValue=<rgba 17.2588 6 20 1>}.RGB, High=@n27.RGB)
  )
  n29 = Constant() {r=0.4}
  n30 = Constant() {r=1}
  n31 = ScalarParameter() {defaultValue=6, parameterName=InnerGlow-Zap}
  n32 = ScalarParameter() {parameterName=Zap}
  n33 = ScalarParameter() {parameterName=Zap2}
  n34 = Add(A=Add(A=Multiply(A=@n32) {constB=2}, B=Multiply(A=@n33) {constB=1.5}), B=Constant() {r=0.5})
  n35 = VectorParameter() {parameterName=GlowColor_Unstable_2, defaultValue=<rgba 0.2191 0 0.4 1>}
  n36 = Saturate(None=Add(A=@n14.R, B=ScalarParameter() {defaultValue=-0.2, parameterName=4444235}))
  n37 = Power(
   Base=@n15
   Exp=LinearInterpolate(A=ScalarParameter() {defaultValue=5, parameterName=4444235_3}, B=ScalarParameter() {defaultValue=0.1, parameterName=4444235_4}, Alpha=@n26)
  )
  n38 = LinearInterpolate(A=ScalarParameter() {defaultValue=1200, parameterName=Radius1}, B=ScalarParameter() {defaultValue=1600, parameterName=Radius2}, Alpha=@n26)
  n39 = Subtract(A=CameraPositionWS(), B=@n1)
  n40 = Distance(A=@n39, B=Constant())
  n41 = Multiply(A=@n40, B=@n40)
  n42 = DotProduct(A=@n39, B=CameraVectorWS())
  n43 = Multiply(A=@n42, B=@n42)
  n44 = Max(A=Subtract(A=Multiply(A=@n38, B=@n38) {desc="r^2"}, B=Subtract(A=@n41, B=@n43))) {constB=0}
  n45 = Divide(A=Multiply(A=SquareRoot(None=@n44)) {constB=2}, B=Multiply(A=@n38) {constB=2})
  n46 = ScalarParameter() {defaultValue=0.5, parameterName=SMaskRad1_1}
  n47 = Multiply(A=@n10, B=@n22)
  n48 = Multiply(
   A=Saturate(None=Add(A=@n46, B=@n47))
   B=MaterialFunctionCall(Transition Phase=@n45, Height Texture=Multiply(A=@n36, B=@n37)) {materialFunction=&HeightLerp}.Lerp Alpha No Contrast
  )
  n49 = Add(
   A=ScalarParameter() {defaultValue=-0.5, parameterName=gpow3_1}
   B=Multiply(A=ScalarParameter() {defaultValue=3, parameterName=gpow3_2}, B=@n47)
  )
  n50 = Multiply(A=Multiply(A=Saturate(None=@n49), B=Constant() {r=0.75}), B=Constant() {r=0.5})
  n51 = MaterialFunctionCall(
   1=VectorParameter() {parameterName=GlowColor_4, defaultValue=<rgba 0.045 0.004 0.3 1>}.RGB
   2=LinearInterpolate(
    A=VectorParameter() {parameterName=GlowColor_3, defaultValue=<rgba 0.0712 0.002 0.4 1>}.RGB
    B=VectorParameter() {parameterName=GlowColor_Unstable_3, defaultValue=<rgba 0.1095 0 0.2 1>}.RGB
    Alpha=@n20
   )
   3=LinearInterpolate(A=VectorParameter() {parameterName=GlowColor_2, defaultValue=<rgba 0.0938 0 0.0746 1>}.RGB, B=Multiply(A=@n35.RGB, B=@n26), Alpha=@n20)
   4=@n21
   A=Saturate(None=Add(A=@n48, B=@n50))
  ) {materialFunction=&Lerp_Multiple_Float3}
  n52 = Multiply(A=@n37, B=ScalarParameter() {defaultValue=0.5, parameterName=4444235_1})
  n53 = Power(Base=@n45, Exp=ScalarParameter() {defaultValue=15, parameterName=4444235_6})
  n54 = LinearInterpolate(
   A=@n51.Lerp 4 Inputs
   B=Multiply(A=LinearInterpolate(A=@n30, B=@n31, Alpha=@n34), B=@n51.Lerp 4 Inputs)
   Alpha=Saturate(None=Add(A=@n52, B=@n53))
  )
  n55 = LinearInterpolate(A=Multiply(A=ScalarParameter() {defaultValue=0.25, parameterName=InnerGlowWhileOff}, B=@n54), B=@n54, Alpha=@n34)
  n56 = Power(Base=@n45, Exp=ScalarParameter() {defaultValue=3, parameterName=4444235_2})
  n57 = Multiply(
   A=ScalarParameter() {defaultValue=2, parameterName=InnerGlowWhileOn_2}
   B=LinearInterpolate(
    A=VectorParameter() {parameterName=GlowColor_One_LQ_1, defaultValue=<rgba 0.0534 0.0015 0.3 1>}.RGB
    B=VectorParameter() {parameterName=GlowColor_One_LQ_2, defaultValue=<rgba 0.4299 0 0.6 1>}.RGB
    Alpha=MaterialFunctionCall(Transition Phase=@n56, Height Texture=@n36) {materialFunction=&HeightLerp}.Lerp Alpha No Contrast
   )
  )
  n58 = Multiply(A=ScalarParameter() {defaultValue=0.1, parameterName=InnerGlowWhileOff-LQ}, B=@n57)
  n59 = Add(
   A=Add(A=Multiply(A=@n16, B=@n21), B=Multiply(A=@n28, B=@n29))
   B=QualitySwitch(Default=@n55, Low=LinearInterpolate(A=@n58, B=@n57, Alpha=@n34), High=@n55)
  )
  Reroute(None=Multiply(A=@n18.R, B=Multiply(A=@n19, B=@n59)))
  n60 = LinearInterpolate(A=Constant() {r=0.2}, B=Constant() {r=0.1}, Alpha=@n14.R)
  Reroute(None=Multiply(A=@n18.R, B=Multiply(A=@n19, B=@n60)))
  n61 = VectorParameter() {parameterName=BaseDiffuse, defaultValue=<rgba 0.015 0.0013 0.1 1>}
  Reroute(None=Multiply(A=@n18.R, B=Multiply(A=@n61.RGB, B=@n19)))
  Reroute(None=Constant() {r=0.2})
 ]
}
