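# Two-sided lit material for the prototype purple Cube, layering glowing cracks, edges and corners over a pulsing inner glow.
# UV2 masks shape corner glow and rotated crack textures, while a world-height blend driven by CubeHappening shifts their intensity.
# Camera-ray sphere math fakes internal volume, height-blended with bump-offset noise into a four-color purple ramp.
# Vertex red and two-sided sign mask surface values; Zap parameters feed an inner-glow blend.
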Material /Figment_S06/Athena/Prototype/Blueprints/Cube/Materials/M_Cube_Purple_P3
graph {
 properties {pixelDepthOffsetMode=PDOM_Legacy, twoSided=true, usedShadingModels=Default Lit}
 nodes [
  n1 = ObjectPositionWS()
  Add(
   A=Add(A=@n1, B=VectorParameter() {parameterName=offz, defaultValue=<rgba 0 0 200 1>}.RGB)
   B=VectorParameter() {desc=For zapping players, parameterName=SphereOffset}.RGB
  )
  n2 = Multiply(A=ScalarParameter() {defaultValue=0.65, parameterName=CrackTile_2}, B=TextureCoordinate() {coordinateIndex=2})
  n3 = ScalarParameter() {defaultValue=1, parameterName=CrackTile_5}
  n4 = Rotator(
   Coordinate=Multiply(A=ScalarParameter() {defaultValue=0.75, parameterName=CrackTile_6}, B=TextureCoordinate() {coordinateIndex=2})
   Time=ScalarParameter() {defaultValue=-3, parameterName=CrackTile_7}
  )
  n5 = Add(
   A=TextureSample(UVs=Rotator(Coordinate=@n2, Time=@n3)) {texture=/Game/Effects/Fort_Effects/Textures/Tile/T_Glass_Cracks_M}.G
   B=Multiply(
    A=TextureSample(UVs=@n4) {texture=&T_Glass_Cracks_M}.G
    B=ScalarParameter() {defaultValue=0.15, parameterName=CornerGlow-Power_8}
   )
  )
  n6 = ScalarParameter() {defaultValue=6, parameterName=CornerGlow-Power_6}
  n7 = ScalarParameter() {defaultValue=3, parameterName=CornerGlow-Power_9}
  n8 = Subtract(A=WorldPosition().XYZ, B=ObjectPositionWS())
  n9 = Subtract(
   A=Divide(A=ComponentMask(None=@n8) {b=true}, B=ScalarParameter() {defaultValue=2767.5127, parameterName=CubeHeight})
   B=ScalarParameter() {defaultValue=0.5, parameterName=Bias}
  )
  n10 = ScalarParameter() {defaultValue=1, parameterName=BlendWidth}
  n11 = LinearInterpolate(B=Add(A=Divide(B=@n10)), Alpha=ScalarParameter() {defaultValue=1, parameterName=CubeHappening, sliderMax=1})
  n12 = Saturate(None=Multiply(A=Add(A=@n9, B=@n11), B=@n10) {constB=4})
  n13 = TextureCoordinate() {coordinateIndex=2}
  n14 = Constant() {r=0.5}
  n15 = Constant() {r=1}
  n16 = Constant()
  n17 = TextureCoordinate() {coordinateIndex=2}
  n18 = Constant2Vector() {g=0.5, r=0.5}
  n19 = MaterialFunctionCall(Float2=Abs(None=Subtract(A=@n17, B=@n18))) {materialFunction=&BreakOutFloat2Components}
  n20 = Max(A=@n19.R, B=@n19.R)
  n21 = OneMinus(None=Multiply(A=OneMinus(None=@n20), B=SphereMask(A=@n13, B=@n14, Radius=@n15, Hardness=@n16) {attenuationRadius=0.7, hardnessPercent=30}))
  n22 = Multiply(A=@n5, B=Power(Base=@n21, Exp=LinearInterpolate(A=@n6, B=@n7, Alpha=@n12)))
  n23 = LinearInterpolate(
   A=ScalarParameter() {defaultValue=1, parameterName=CornerGlow-Power_11}
   B=ScalarParameter() {defaultValue=75, parameterName=CornerGlow-Power_12}
   Alpha=@n12
  )
  n24 = MaterialFunctionCall(Value=Multiply(A=Time(), B=ScalarParameter() {defaultValue=1, parameterName=MainPulseSpeed})) {materialFunction=&LinearSine}
  n25 = LinearInterpolate(
   A=ScalarParameter() {defaultValue=0.75, parameterName=CornerGlow-PulseA}
   B=ScalarParameter() {defaultValue=1, parameterName=CornerGlow-PulseA_1}
   Alpha=@n24."Rounded Linear Sine "
  )
  n26 = Multiply(A=Power(Base=@n21, Exp=ScalarParameter() {defaultValue=8, parameterName=CornerGlow-Power_5}), B=@n25)
  n27 = Multiply(A=Power(Base=@n21, Exp=ScalarParameter() {defaultValue=2, parameterName=CornerGlow-Power_2}), B=@n25)
  n28 = Multiply(A=Constant() {r=3.5}, B=TextureCoordinate())
  n29 = TextureSample(UVs=BumpOffset(Coordinate=@n28, Height=Constant() {r=0.58}, HeightRatioInput=Constant() {r=-0.5})) {samplerType=SAMPLERTYPE_Masks, texture=&/T_CubeNoise}
  n30 = TextureSample(UVs=@n28) {samplerType=SAMPLERTYPE_Masks, texture=&T_CubeNoise}
  n31 = QualitySwitch(Default=@n29.G, Low=@n30.G, High=@n29.G)
  n32 = Add(
   A=@n26
   B=Multiply(
    A=@n26
    B=MaterialFunctionCall(Transition Phase=@n27, Height Texture=@n31) {materialFunction=&HeightLerp}.Lerp Alpha No Contrast
   )
  )
  n33 = Multiply(A=@n32, B=ScalarParameter() {defaultValue=8, parameterName=CornerGlow-Power_1})
  n34 = ScalarParameter() {defaultValue=3, parameterName=CornerGlow-Power_7}
  n35 = ScalarParameter() {defaultValue=20, parameterName=CornerGlow-Power_10}
  n36 = LinearInterpolate(
   A=ScalarParameter() {defaultValue=-3.85, parameterName=CrackEdgeMaskLerpA}
   B=ScalarParameter() {defaultValue=4.1, parameterName=CrackEdgeMaskLerpB}
   Alpha=@n20
  )
  n37 = Add(
   A=Add(A=Multiply(A=@n22, B=@n23), B=Multiply(A=@n12, B=@n33))
   B=Multiply(A=LinearInterpolate(A=@n34, B=@n35, Alpha=@n12), B=Saturate(None=@n36))
  )
  n38 = ScalarParameter() {defaultValue=4, parameterName=CornerGlow-ModLast}
  n39 = ScalarParameter() {defaultValue=2, parameterName=CornerGlow-Power_3}
  n40 = Add(A=@n30.R, B=ScalarParameter() {defaultValue=0.5, parameterName=4444235})
  n41 = LinearInterpolate(
   A=ScalarParameter() {defaultValue=5, parameterName=PulseTexturePowerA}
   B=ScalarParameter() {defaultValue=3, parameterName=PulseTexturePowerB}
   Alpha=@n24."Rounded Linear Sine "
  )
  n42 = ScalarParameter() {defaultValue=1200, parameterName=Radius2_1}
  n43 = Multiply(A=@n42, B=@n42) {desc="r^2"}
  n44 = Subtract(A=CameraPositionWS(), B=@n1)
  n45 = Distance(A=@n44, B=Constant())
  n46 = DotProduct(A=@n44, B=CameraVectorWS())
  n47 = Subtract(A=Multiply(A=@n45, B=@n45), B=Multiply(A=@n46, B=@n46))
  n48 = SquareRoot(None=Max(A=Subtract(A=@n43, B=@n47)) {constB=0})
  n49 = MaterialFunctionCall(
   Transition Phase=Divide(A=Multiply(A=@n48) {constB=2}, B=Multiply(A=@n42) {constB=2})
   Height Texture=Multiply(A=Saturate(None=@n40), B=Power(Base=@n31, Exp=@n41))
  ) {materialFunction=&HeightLerp}
  n50 = MaterialFunctionCall(
   1=VectorParameter() {parameterName=GlowColor1, defaultValue=<rgba 0.0201 0 0.05 1>}.RGB
   2=VectorParameter() {parameterName=GlowColor2, defaultValue=<rgba 0.0804 0 0.2 1>}.RGB
   3=VectorParameter() {parameterName=GlowColor3, defaultValue=<rgba 0.1475 0 0.35 1>}.RGB
   4=VectorParameter() {parameterName=GlowColor4, defaultValue=<rgba 0.2866 0 0.4 1>}.RGB
   A=Add(A=Multiply(A=@n32, B=@n39), B=@n49.Lerp Alpha No Contrast)
  ) {materialFunction=&Lerp_Multiple_Float3}
  n51 = Multiply(A=@n32, B=@n50.Lerp 4 Inputs)
  n52 = Power(Base=@n49.Lerp Alpha No Contrast, Exp=ScalarParameter() {defaultValue=10, parameterName=4444235_4})
  n53 = Add(
   A=Add(A=Multiply(A=@n38, B=@n51), B=@n50.Lerp 4 Inputs)
   B=Multiply(A=Multiply(A=@n50.Lerp 4 Inputs, B=@n52), B=ScalarParameter() {defaultValue=1, parameterName=CornerGlow-Power_4})
  )
  Add(A=Multiply(A=@n37, B=@n53), B=@n53)
  n54 = Rotator(
   Coordinate=Multiply(A=ScalarParameter() {defaultValue=1, parameterName=CrackTile_1}, B=TextureCoordinate() {coordinateIndex=2})
   Time=ScalarParameter() {defaultValue=-4.1069, parameterName=CrackTile_3}
  )
  Add(
   A=Multiply(
    A=ScalarParameter() {defaultValue=2, parameterName=CubeHappening_1, sliderMax=1}
    B=TextureSample(UVs=@n54) {texture=&T_Glass_Cracks_M}.G
   )
   B=@n5
  )
  Constant3Vector() {constant=<rgba 0 0 -400 1>}
  Constant() {r=5}
  Constant() {r=0.1}
  Constant()
  Constant() {r=8}
  Constant() {r=10}
  Constant() {r=5}
  Constant() {r=0.75}
  Constant() {r=1}
  Constant() {r=1200}
  Constant() {r=1400}
  LinearInterpolate(
   A=ScalarParameter() {defaultValue=1200, parameterName=Radius1}
   B=ScalarParameter() {defaultValue=1200, parameterName=Radius2}
   Alpha=@n24."Rounded Linear Sine "
  )
  n55 = ScalarParameter() {parameterName=Zap}
  n56 = ScalarParameter() {parameterName=Zap2}
  LinearInterpolate(
   A=Constant() {r=1}
   B=ScalarParameter() {defaultValue=6, parameterName=InnerGlow-Zap}
   Alpha=Add(A=Multiply(A=@n55) {constB=2}, B=Multiply(A=@n56) {constB=1.5})
  )
  n57 = VectorParameter() {parameterName=GlowColor4_1, defaultValue=<rgba 0.2866 0 0.4 1>}
  n58 = LinearInterpolate(
   A=ScalarParameter() {defaultValue=1, parameterName=CornerGlow-Power_13}
   B=ScalarParameter() {defaultValue=0.3, parameterName=CornerGlow-Power_14}
   Alpha=@n12
  )
  Multiply(A=Add(A=Multiply(A=@n57.RGB, B=@n37), B=Multiply(A=@n58, B=@n53)))
  Reroute(None=LinearInterpolate(A=Constant() {r=0.29}, B=Constant() {r=0.8}, Alpha=@n30.R))
  n59 = VertexColor()
  n60 = TwoSidedSign()
  n61 = LinearInterpolate(A=Constant() {r=0.2}, B=Constant() {r=0.1}, Alpha=@n30.R)
  Reroute(None=Multiply(A=@n59.R, B=Multiply(A=@n60, B=@n61)))
  Reroute(None=Constant() {r=0.2})
  n62 = VectorParameter() {parameterName=BaseDiffuse, defaultValue=<rgba 0.015 0.0013 0.1 1>}
  Reroute(None=Multiply(A=@n59.R, B=Multiply(A=@n62.RGB, B=@n60)))
  Reroute(None=Multiply(A=@n59.R, B=Multiply(A=@n60)))
 ]
}
